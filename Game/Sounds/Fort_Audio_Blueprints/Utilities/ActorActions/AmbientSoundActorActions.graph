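# Editor utility blueprint for audio work that selects AmbientSound actors within a given Distance of the first selected actor in the level.
# SelectNearbyAmbientSoundActors filters all level actors by AmbientSound class, then compares each one's distance to that selected actor, re-reading the selection on every loop pass.
# Matches collect in NearbyActors, which replaces the editor selection once the loop completes.
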EditorUtilityBlueprint /Game/Sounds/Fort_Audio_Blueprints/Utilities/ActorActions/AmbientSoundActorActions
graphs [
 EventGraph {links [], nodes []}
 SelectNearbyAmbientSoundActors {nodes [
   SelectNearbyAmbientSoundActors!() -> (then, Distance:double=~) {is=title K2Node_FunctionEntry}
   > n1 = GetAllLevelActors(self=Get EditorActorSubsystem())
   > n2 = Filter by Class!(
    execute
    TargetArray:Array of Object References=@n1
    ObjectClass:Object Class Reference=/Script/Engine.AmbientSound
    FilterType:EEditorScriptingFilterType Enum
   ) -> (then, ReturnValue:Array of Ambient Sound Object References)
   > n3 = For Each Loop(Array:Array of Ambient Sound Object References=@n2) -> (Array Element:Ambient Sound Object Reference)
   SetSelectedLevelActors(
    execute=@n3.Completed
    self=Get EditorActorSubsystem()
    ActorsToSelect=Get NearbyActors() -> (NearbyActors:Array of Ambient Sound Object References)
   )
   n4 = GetSelectedLevelActors(execute=@n3.LoopBody, self=Get EditorActorSubsystem())
   n5 = "Get (a copy)"(Array:Array of Actor Object References=@n4) -> (Output:Actor Object Reference)
   > Branch(Condition="float < float"(
     A=GetDistanceBetweenActors!(Actor1:Actor Object Reference=@n5, Actor2:Actor Object Reference=@n3, __WorldContext:Object Reference=@n3) -> (Distance:double)
     B=Get Distance() -> (Distance:double)
    ))
   > Add(TargetArray:Array of Ambient Sound Object References=Get NearbyActors(), NewItem:"Ambient Sound Object Reference (by ref)"=@n3) {is=title K2Node_CallArrayFunction}
  ]}
]
variables []
