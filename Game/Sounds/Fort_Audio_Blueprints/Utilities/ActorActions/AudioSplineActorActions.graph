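# Editor utility actor action that splits the spline of selected actors into a given number of audio-only ocean water spline actors.
# It caches world point locations, spawns BP_WaterSpline_AudioOnly_Ocean actors with open splines, then moves each to its starting point.
# Consecutive index ranges, sized by dividing point count by split count, are added as world-space points.
# A branch copying full spline point structs is present but unwired.
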EditorUtilityBlueprint /Game/Sounds/Fort_Audio_Blueprints/Utilities/ActorActions/AudioSplineActorActions
graphs [
 EventGraph {links [], nodes []}
 SplitSplineIntoAudioOnlySegments {nodes [
   SplitSplineIntoAudioOnlySegments!() -> (then, Split Count:int=2) {is=title K2Node_FunctionEntry}
   > n1 = GetSelectedLevelActors(self=Get EditorActorSubsystem())
   > n2 = For Each Loop(Array:Array of Actor Object References=@n1) -> (Array Element:Actor Object Reference)
   > n3 = Set OriginalSpline(OriginalSpline:Spline Component Object Reference=GetComponentByClass!(self:Actor Object Reference=@n2, ComponentClass:Actor Component Class Reference=/Script/Engine.SplineComponent) -> (ReturnValue:Spline Component Object Reference)) -> (Output_Get:Spline Component Object Reference)
   > n4 = For Loop(LastIndex=int - int(A=GetNumberOfSplinePoints!(self:Spline Component Object Reference=@n3) -> (ReturnValue:int=0)))
   n5 = For Loop(execute=@n4.Completed, LastIndex=int - int(A=Get Split Count() -> (Split Count:int=0)))
   n6 = Get SplineLocations() -> (SplineLocations:Array of Vectors)
   n7 = float / float(A="To Float (Integer)"(InInt=Length(TargetArray:Array of Vectors=@n6)), B="To Float (Integer)"(InInt=Get Split Count() -> (Split Count=0)))
   Set bRemainder(execute=@n5.Completed, bRemainder:bool="Not Equal (Float)"(A=Fraction!(A:double=@n7) -> (ReturnValue:double))) -> (Output_Get:bool)
   > n8 = Set SplinePointsPerComponent(SplinePointsPerComponent:int=Truncate(A=@n7)) -> (Output_Get:int=0)
   > Set EndingIndex(EndingIndex:int=@n8) -> (Output_Get:int=0)
   n9 = Get SplitSplineComponents() -> (SplitSplineComponents:Array of Spline Component Object References)
   > n10 = For Each Loop(Array:Array of Spline Component Object References=@n9) -> (Array Element:Spline Component Object Reference)
   > TransactObject(Object=@n10)
   n11 = "Get (a copy)"(Array:Array of Vectors=Get SplineLocations(), Dimension 1=Get StartingIndex() -> (StartingIndex:int=0)) -> (Output:Vector)
   > Set Actor Location!(execute, self:Actor Object Reference=GetOwner(self=@n10), NewLocation:Vector=@n11, bSweep:bool=false, bTeleport:bool=false) -> (then, SweepHitResult:Hit Result Structure, ReturnValue:bool) {is=title K2Node_CallFunction}
   > Set World Location!(execute, self:Scene Component Object Reference=@n10, NewLocation:Vector=@n11, bSweep:bool=false, bTeleport:bool=false) -> (then, SweepHitResult:Hit Result Structure) {is=title K2Node_CallFunction}
   > n12 = For Loop(FirstIndex=Get StartingIndex() -> (StartingIndex=0), LastIndex=Get EndingIndex() -> (EndingIndex:int=0))
   UpdateSpline(execute=@n12.Completed, self:Spline Component Object Reference=@n10)
   > Set StartingIndex(StartingIndex:int=int + int!(A=Get StartingIndex() -> (StartingIndex=0), B=Get SplinePointsPerComponent() -> (SplinePointsPerComponent:int=0), C:int=1) -> (ReturnValue)) -> (Output_Get:int=0)
   n13 = AND Boolean(
    A=Get bRemainder() -> (bRemainder:bool)
    B="Equal (Integer)"(B=Last Index!(TargetArray:Array of Spline Component Object References=@n9) -> (ReturnValue=0))
   )
   > Set EndingIndex(EndingIndex=int + int(
     A=Get StartingIndex() -> (StartingIndex=0)
     B=Get SplinePointsPerComponent() -> (SplinePointsPerComponent=0)
     C=SelectInt(A=1, B=0, bPickA=@n13)
    )) -> (Output_Get=0)
   n14 = Get SplineLocations()
   Branch(execute=@n12.LoopBody, Condition=Is Valid Index!(TargetArray:Array of Vectors=@n14, IndexToTest:int=@n12) -> (ReturnValue:bool))
   > AddSplinePoint!(
    execute
    self:Spline Component Object Reference=@n10
    Position:"Vector (by ref)"="Get (a ref)"(Array:Array of Vectors=@n14, Dimension 1=@n12) -> (Output:"Vector (by ref)")
    CoordinateSpace:ESplineCoordinateSpace Enum=World
    bUpdateSpline:bool=true
   ) -> (then)
   n15 = SpawnActorFromClass!(
    execute=@n5.LoopBody
    self:Editor Actor Subsystem Object Reference=Get EditorActorSubsystem()
    ActorClass:Actor Class Reference=/Game/Athena/Sounds/Water/BP_WaterSpline_AudioOnly_Ocean.BP_WaterSpline_AudioOnly_Ocean_C
    Location:Vector="0, 0, 0"
    Rotation:Rotator
    bTransient:bool
   ) -> (then, ReturnValue:BP Water Spline Audio Only Ocean Object Reference)
   n16 = GetComponentByClass(self=@n15, ComponentClass=&SplineComponent)
   > Add(
    TargetArray:Array of Spline Component Object References=Get SplitSplineComponents()
    NewItem:"Spline Component Object Reference (by ref)"=@n16
   ) {is=title K2Node_CallArrayFunction}
   > SetClosedLoop!(execute, self:Spline Component Object Reference=@n16, bInClosedLoop:bool=false, bUpdateSpline:bool=true) -> (then)
   n17 = Get OriginalSpline() -> (OriginalSpline:Spline Component Object Reference)
   Add(
    execute=@n4.LoopBody
    TargetArray:Array of Vectors=Get SplineLocations()
    NewItem:"Vector (by ref)"=GetLocationAtSplinePoint!(self:Spline Component Object Reference=@n17, PointIndex:int=@n4, CoordinateSpace:ESplineCoordinateSpace Enum=World) -> (ReturnValue:Vector)
   )
   > Add(
    TargetArray:Array of Spline Point Structures=Get SplinePoints() -> (SplinePoints:Array of Spline Point Structures)
    NewItem:"Spline Point Structure (by ref)"=GetSplinePointAt!(self:Spline Component Object Reference=@n17, PointIndex:int=@n4, CoordinateSpace:ESplineCoordinateSpace Enum=World) -> (ReturnValue:Spline Point Structure)
   )
   n18 = Get SplinePoints()
   Branch(Condition=Is Valid Index(TargetArray:Array of Spline Point Structures=@n18, IndexToTest=@n12))
   > AddPoint!(
    execute
    self:Spline Component Object Reference=@n10
    Point:"Spline Point Structure (by ref)"="Get (a copy)"(Array:Array of Spline Point Structures=@n18, Dimension 1=@n12) -> (Output:Spline Point Structure)
    bUpdateSpline:bool=false
   ) -> (then)
  ]}
]
variables []
